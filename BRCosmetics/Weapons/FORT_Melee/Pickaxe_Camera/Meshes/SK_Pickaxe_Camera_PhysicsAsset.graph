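# Physics asset for the skeletal mesh of the Camera pickaxe, a two-handed melee weapon cosmetic.
# It approximates the weapon with two capsule bodies: a narrow one on the Melee_TwoHanded_Handle bone and a much larger one, offset and tilted, on the b_AR_Root bone.
# A single constraint joins the two bodies, with matching swing and twist limit angles.
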PhysicsAsset /BRCosmetics/Weapons/FORT_Melee/Pickaxe_Camera/Meshes/SK_Pickaxe_Camera_PhysicsAsset
bodies {
 Melee_TwoHanded_Handle [{shapeType=Capsule, length=26.1116, radius=4.471, center=<0.0276 -1.1579 -19.4879>, rotation=<rot -3.1076 -90 0.0937>}]
 b_AR_Root [{shapeType=Capsule, length=35.8407, radius=27.1536, center=<7.409 -0.4072 49.307>, rotation=<rot 0 90.443 -58.5406>}]
}
constraints [{bone1Name=Melee_TwoHanded_Handle, bone2Name=b_AR_Root, swing1LimitDegrees=45, swing2LimitDegrees=45, twistLimitDegrees=45}]
